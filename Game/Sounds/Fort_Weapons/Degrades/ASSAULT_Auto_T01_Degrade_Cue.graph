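# Sound cue for the degraded-state loop of the tier 1 automatic assault weapon, routed through the ranged weapon low-ammo/degraded sound class.
# A looping wave player plays the auto_degrade_loop_01 wave through a Modulator node at the root, with boosted cue volume.
# A concurrency override prevents new instances from starting once the limit is reached.
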SoundCue /Game/Sounds/Fort_Weapons/Degrades/ASSAULT_Auto_T01_Degrade_Cue
root=Modulator_0
nodes [
 {id=Modulator_0, children [WavePlayer_0]}
 {id=WavePlayer_0, properties {bLooping=true, soundWaveAssetPtr=&/auto_degrade_loop_01}}
]
properties {
 bOverrideConcurrency=true
 maxDistance=2097152
 soundClassObject=/Game/Sounds/FortSoundClassesAndModes/Fort_WeaponRanged_LowAmmo_Degraded_PSM
 volumeMultiplier=1.5
 concurrencyOverrides {resolutionRule=PreventNew}
}
