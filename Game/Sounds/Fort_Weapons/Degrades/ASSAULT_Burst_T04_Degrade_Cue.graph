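# Sound cue for the degraded-weapon sound of the tier 4 burst assault rifle, routed to the low-ammo/degraded ranged weapon sound class.
# A quality-level switch feeds two modulated random branches: the first picks from four rifle degrade waves, the second from three of the same waves, with the other two quality slots empty.
# Its own concurrency override blocks new instances once the limit is reached, and a volume multiplier boosts the output.
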SoundCue /Game/Sounds/Fort_Weapons/Degrades/ASSAULT_Burst_T04_Degrade_Cue
root=QualityLevel_0
nodes [
 {id=Modulator_1, children [Random_1]}
 {id=Modulator_2, children [Random_2]}
 {id=QualityLevel_0, children [Modulator_1,Modulator_2,null,null]}
 {id=Random_1, children [WavePlayer_4,WavePlayer_5,WavePlayer_6,WavePlayer_7], properties {weights [1,1,1,1]}}
 {id=Random_2, children [WavePlayer_9, WavePlayer_10, WavePlayer_11], properties {weights [1,1,1]}}
 {id=WavePlayer_10, properties {soundWaveAssetPtr=/Game/Sounds/Fort_Weapons/pistol/degrades/rifle_degrade_03}}
 {id=WavePlayer_11, properties {soundWaveAssetPtr=/Game/Sounds/Fort_Weapons/pistol/degrades/rifle_degrade_04}}
 {id=WavePlayer_4, properties {soundWaveAssetPtr=/Game/Sounds/Fort_Weapons/pistol/degrades/rifle_degrade_01}}
 {id=WavePlayer_5, properties {soundWaveAssetPtr=/Game/Sounds/Fort_Weapons/pistol/degrades/rifle_degrade_02}}
 {id=WavePlayer_6, properties {soundWaveAssetPtr=&rifle_degrade_03}}
 {id=WavePlayer_7, properties {soundWaveAssetPtr=&rifle_degrade_04}}
 {id=WavePlayer_9, properties {soundWaveAssetPtr=&rifle_degrade_02}}
]
properties {
 bOverrideConcurrency=true
 maxDistance=2097152
 soundClassObject=/Game/Sounds/FortSoundClassesAndModes/Fort_WeaponRanged_LowAmmo_Degraded_PSM
 volumeMultiplier=1.5
 concurrencyOverrides {resolutionRule=PreventNew}
}
