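# Sound cue for the degraded-weapon layer of the standard T00 rifle, routed to the ranged weapon low-ammo/degraded sound class.
# A Random node picks one of six equally weighted small single-projectile durability waves.
# A Modulator then raises it to a randomized higher pitch, and the cue boosts volume.
# A concurrency override prevents new instances from starting once the limit is reached.
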SoundCue /Game/Sounds/Fort_Weapons/Degrades/Rifle_Standard_T00_Degrade_Cue
root=Modulator_21
nodes [
 {id=Modulator_21, children [Random_23], properties {pitchMax=1.4, pitchMin=1.3}}
 {
  id=Random_23
  children [WavePlayer_137, WavePlayer_138, WavePlayer_139, WavePlayer_140, WavePlayer_141, WavePlayer_142]
  properties {weights [1,1,1,1,1,1]}
 }
 {id=WavePlayer_137, properties {soundWaveAssetPtr=&/fort_projectile_single_small_dur_01}}
 {id=WavePlayer_138, properties {soundWaveAssetPtr=&/fort_projectile_single_small_dur_02}}
 {id=WavePlayer_139, properties {soundWaveAssetPtr=&/fort_projectile_single_small_dur_03}}
 {id=WavePlayer_140, properties {soundWaveAssetPtr=&/fort_projectile_single_small_dur_04}}
 {id=WavePlayer_141, properties {soundWaveAssetPtr=&/fort_projectile_single_small_dur_05}}
 {id=WavePlayer_142, properties {soundWaveAssetPtr=&/fort_projectile_single_small_dur_06}}
]
properties {
 bOverrideConcurrency=true
 maxDistance=2097152
 soundClassObject=/Game/Sounds/FortSoundClassesAndModes/Fort_WeaponRanged_LowAmmo_Degraded_PSM
 volumeMultiplier=1.5
 concurrencyOverrides {resolutionRule=PreventNew}
}
